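; Arrow projectile for the Figment sniper crossbow, taking its gravity from a scalable float at construction.
; Blocking touches stop gravity and trail effects, destroy non-teammate bushes or damage snowmen and balloons, then explode where cosmetics don't run (dedicated server).
; Explosions play player or surface impact sounds.
; Water overlaps and sniper-trace water hits hide the trail and spawn splash sound and depth-parameterized VFX; the trace path retains a debug print.
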(event EventOnExploded
  (params (HitActors :type Array%20of%20Actor%20Object%20References) (HitResults :type Array%20of%20Hit%20Result%20Structures))
  (bind (_bblockinghit _binitialoverlap _time _distance _location _impactpoint _normal _impactnormal _physmat _hitactor) (Collision|BreakHitResult (Variables|Default|GetHitResult)))
  (bind (_asfort_player_pawn _bsuccess) (Utilities|Casting|CastToFortPlayerPawn _hitactor))
  (Networking|CanExecuteCosmeticEvents
    (:True
      (Audio|PlaySoundatLocation (select _bsuccess "/Figment_Sniper_Crossbow_Athena/Gameplay/Sounds/Crossbow_Player_Impact_Athena_Figment_Cue.Crossbow_Player_Impact_Athena_Figment_Cue" "/Figment_Sniper_Crossbow_Athena/Gameplay/Sounds/Crossbow_Surface_Impact_Athena_Figment_Cue.Crossbow_Surface_Impact_Athena_Figment_Cue") _impactpoint))))

(event EventOnTouched
  (params (OtherActor :type Actor%20Object%20Reference) (OtherComp :type Primitive%20Component%20Object%20Reference) (HitResult :type Hit%20Result%20Structure%20%28by%20ref%29) (bIsOverlap :type Boolean))
  (bind (_bblockinghit _binitialoverlap _time _distance _location _impactpoint _normal _impactnormal _physmat _hitactor) (Collision|BreakHitResult HitResult))
  (bind _waterhit (Variables|Default|GetWaterHit))
  (bind (_bblockinghit_2 _binitialoverlap_2 _time_2 _distance_2 _location_2 _impactpoint_2 _normal_2 _impactnormal_2 _physmat_2 _hitactor_2) (Collision|BreakHitResult _waterhit))
  (bind _waterinteractioncomponent (Variables|Water|GetWaterInteractionComponent))
  (bind (_waterplanelocation _waterplanenormal _watersurfaceposition _waterdepth _waterbodyidx) (FX|GetLastWaterSurfaceInfo _waterinteractioncomponent :!class /Script/FortniteGame.FortWaterInteractionComponent))
  (Variables|Default|SetHitResult HitResult)
  (if bIsOverlap
    ;[0 0.298 1] Water Impact Projectile
    (Variables|Default|SetWaterHit HitResult)
    (Variables|Default|SetWaterHitMat _physmat_2)
    (Utilities|Casting|CastToFortWaterBodyActor _hitactor_2)
    (Rendering|SetVisibility (Variables|Default|GetP_Sniper_Bullet_Trail) false true)
    (CallFunction|PlayImpactSFX false true)
    (bind _returnvalue (Effects|SpawnEmitteratLocation (Variables|Default|GetWaterImpactVFX) "0, 0, 0" (Math|Conversions|RotationFromXVector _waterplanenormal)))
    (Effects|Components|ParticleSystem|SetFloatParameter _returnvalue "WaterDepth" _waterdepth)
    (Effects|Components|ParticleSystem|SetFloatParameter _returnvalue "WaterBodyIndex" (Math|Conversions|ToFloat(Integer) _waterbodyidx))
    (else
      (Components|Activation|SetActive (Variables|Default|GetP_Crossbow_Projectile) false false)
      (Variables|/Script/Engine.ProjectileMovementComponent|SetProjectileGravityScale 0.0 (Variables|Movement|GetProjectileMovementComponent))
      (Class|BulletWhipTrackerComponentBase|Disable (Variables|Visual|GetBulletWhipTrackerComponent))
      (CallFunction|HitBlockingMeshonPlayer? _hitactor)
      (Networking|CanExecuteCosmeticEvents
        (:False
          (Utilities|FlowControl|Delay 0.0)
          (if (not (Delivery|IsBeingKilled))
            (Delivery|DoExplosion)))))))

(event EventOnSniperTraceWaterHit
  (params (WaterHit :type Hit%20Result%20Structure%20%28by%20ref%29) (WaterBody :type Fort%20Water%20Body%20Actor%20Object%20Reference))
  (bind _waterhit (Variables|Default|GetWaterHit))
  (bind (_bblockinghit _binitialoverlap _time _distance _location _impactpoint _normal _impactnormal _physmat) (Collision|BreakHitResult _waterhit))
  (bind _waterdepthtrace (Variables|Default|GetWaterDepthTrace))
  ;[0 0.298 1] Water Impact Trace
  (Variables|Default|SetWaterHit WaterHit)
  (Utilities|FlowControl|Sequence
    (:then_0
      (Utilities|FlowControl|DoOnce false
        (:Completed
          (Variables|Default|SetTraceImpactLocation _location)
          (Variables|Default|SetTraceImpactNormal _impactnormal))))
    (:then_1
      (Networking|CanExecuteCosmeticEvents
        (:True
          (bind _output_get (Variables|Default|SetWaterHitMat _physmat))
          (Utilities|IsValid _output_get
            (:Is%20Valid
              (if (Utilities|Enum|Equal(Enum) (Variables|/Script/PhysicsCore.PhysicalMaterial|GetSurfaceType _output_get) "SurfaceType8")
                (Rendering|SetVisibility (Variables|Default|GetP_Sniper_Bullet_Trail) false false)
                (Utilities|FlowControl|Delay 0.01)
                (CallFunction|PlayImpactSFX false)
                (bind _outwatersurfaceinfo (WaterBody|GetWaterSurfaceInfoStruct WaterBody (Variables|Default|GetTraceImpactLocation) true false false))
                (Variables|Default|SetWaterDepthTrace (.WaterDepth (Utilities|Struct|BreakWaterSurfaceInfo _outwatersurfaceinfo)))
                (Variables|Default|SetWaterIndexTrace (.WaterBodyIdx (Utilities|Struct|BreakWaterSurfaceInfo _outwatersurfaceinfo)))
                (bind _returnvalue (Effects|SpawnEmitteratLocation (Variables|Default|GetWaterImpactVFX) (Variables|Default|GetTraceImpactLocation) (Math|Rotator|MakeRotfromX (Variables|Default|GetTraceImpactNormal))))
                (Development|PrintString (Utilities|String|ToString(Float) _waterdepthtrace))
                (Effects|Components|ParticleSystem|SetFloatParameter _returnvalue "WaterDepth" _waterdepthtrace)
                (Effects|Components|ParticleSystem|SetFloatParameter _returnvalue "WaterBodyIndex" (Math|Conversions|ToFloat(Integer) (Variables|Default|GetWaterIndexTrace)))))))))))

(fn HitBlockingMeshOnPlayer?
  (params (HitActor :type Actor%20Object%20Reference))
  (bind _returnvalue (Actor|GetOwner HitActor))
  (Utilities|IsValid HitActor
    (:Is%20Valid
      (if (not (Team|OnSameTeam _returnvalue (Variables|Actor|GetInstigator)))
        (bind _asathena_player_bush (Utilities|Casting|CastToAthena_Player_Bush HitActor)
          (:then
            (Actor|DestroyActor _asathena_player_bush))
          (:CastFailed
            (bind _asathena_enemy_player_sneaky_snowman (Utilities|Casting|CastToAthena_Enemy_Player_SneakySnowman HitActor)
              (:then
                (Environment|ApplyGlobalEnvironmentGameplayEffecttoActor _asathena_enemy_player_sneaky_snowman "/Game/Weapons/FORT_Sniper/Blueprints/GE_Crossbow_EnvDamage.GE_Crossbow_EnvDamage_C" 1))
              (:CastFailed
                (bind _asbuilding_gameplay_actor_balloon (Utilities|Casting|CastToBuildingGameplayActorBalloon HitActor)
                  (:then
                    (Environment|ApplyGlobalEnvironmentGameplayEffecttoActor _asbuilding_gameplay_actor_balloon "/Game/Weapons/FORT_Sniper/Blueprints/GE_Crossbow_EnvDamage.GE_Crossbow_EnvDamage_C" 1))
                  (:CastFailed
                    (bind _asathena_prop_sneaky_snowman (Utilities|Casting|CastToAthena_Prop_SneakySnowman HitActor)
                      (:then
                        (Environment|ApplyGlobalEnvironmentGameplayEffecttoActor _asathena_prop_sneaky_snowman "/Game/Weapons/FORT_Sniper/Blueprints/GE_Crossbow_EnvDamage.GE_Crossbow_EnvDamage_C" 1)
                        (Components|DestroyComponent (Variables|/Script/Engine.Actor|GetRootComponent _returnvalue)))
                      (:CastFailed
                        _))
                    (return)))))))))
    (:Is%20Not%20Valid
      _)))

(fn PlayImpactSFX
  (params (IsPlayer :type Boolean) (IsWater :type Boolean))
  (bind _returnvalue (Transformation|GetActorLocation))
  (if IsWater
    (Audio|PlaySoundatLocation (Variables|Default|GetSniper_Rifle_WaterImpactSound) _returnvalue)
    (Variables|Default|SetbHasPlayedWaterSound true)
    (elif (Variables|Default|GetbHasPlayedWaterSound)
      _
      (else
        (Audio|PlaySoundatLocation (select IsPlayer (Variables|Default|GetSniper_Rifle_PlayerImpactSound) (Variables|Default|GetSniper_Rifle_SurfaceImpactSound)) _returnvalue)))))

(fn ConstructionScript
  (Projectile|SetGravityScale (ScalableFloat|GetValueatLevel (Variables|Default|GetProjectileGravityScale))))
