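# Sound cue for a crossbow arrow hitting water, built from reused medium bullet water-impact waves.
# A Random node picks one of six waves with equal weight and is excluded from branch culling.
# A Modulator then raises the pitch well above normal within a narrow random range and boosts volume, on top of a high cue volume multiplier.
# It uses the bullet impact attenuation and the impacts sound class.
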SoundCue /Figment_Sniper_Crossbow_Athena/Gameplay/Sounds/CrossbowArrow_Impact_Water_Cue
root=Modulator_0
nodes [
 {id=Modulator_0, children [Random_0], properties {pitchMax=1.8, pitchMin=1.7, volumeMax=2, volumeMin=2}}
 {
  id=Random_0
  children [WavePlayer_5, WavePlayer_6, WavePlayer_7, WavePlayer_8, WavePlayer_9, WavePlayer_10]
  properties {bShouldExcludeFromBranchCulling=true, weights [1,1,1,1,1,1]}
 }
 {id=WavePlayer_10, properties {soundWaveAssetPtr=/Game/Sounds/Fort_Weapons/Impacts/Bullet_Impacts_S10/Bullet_Impact_Water_Med_05}}
 {id=WavePlayer_5, properties {soundWaveAssetPtr=/Game/Sounds/Fort_Weapons/Impacts/Bullet_Impacts_S10/Bullet_Impact_Water_Med_10}}
 {id=WavePlayer_6, properties {soundWaveAssetPtr=/Game/Sounds/Fort_Weapons/Impacts/Bullet_Impacts_S10/Bullet_Impact_Water_Med_01}}
 {id=WavePlayer_7, properties {soundWaveAssetPtr=/Game/Sounds/Fort_Weapons/Impacts/Bullet_Impacts_S10/Bullet_Impact_Water_Med_02}}
 {id=WavePlayer_8, properties {soundWaveAssetPtr=/Game/Sounds/Fort_Weapons/Impacts/Bullet_Impacts_S10/Bullet_Impact_Water_Med_03}}
 {id=WavePlayer_9, properties {soundWaveAssetPtr=/Game/Sounds/Fort_Weapons/Impacts/Bullet_Impacts_S10/Bullet_Impact_Water_Med_04}}
]
properties {
 attenuationSettings=/Game/Sounds/Attenuation/Fort_Bullet_Impact_Att
 maxDistance=5300
 soundClassObject=&Fort_Impacts
 volumeMultiplier=5
}
