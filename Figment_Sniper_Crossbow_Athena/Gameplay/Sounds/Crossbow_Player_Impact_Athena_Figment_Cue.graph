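# Sound cue for crossbow bolt impacts on players, part of the Figment sniper crossbow in Athena.
# A Random node picks among four shared crossbow impact waves, with two empty, equally weighted slots, and feeds a Modulator that varies volume.
# It uses bullet impact attenuation, an impacts sound class and a concurrency override that caps simultaneous plays, stopping the oldest.
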SoundCue /Figment_Sniper_Crossbow_Athena/Gameplay/Sounds/Crossbow_Player_Impact_Athena_Figment_Cue
root=Modulator_0
nodes [
 {id=Modulator_0, children [Random_0], properties {volumeMax=1.5}}
 {id=Random_0, children [WavePlayer_0,WavePlayer_1,WavePlayer_2,WavePlayer_3,null,null], properties {weights [1,1,1,1,1,1]}}
 {id=WavePlayer_0, properties {soundWaveAssetPtr=/FNGameplayCues/Athena/Sounds/Impacts/Crossbow_Impact_01}}
 {id=WavePlayer_1, properties {soundWaveAssetPtr=/FNGameplayCues/Athena/Sounds/Impacts/Crossbow_Impact_02}}
 {id=WavePlayer_2, properties {soundWaveAssetPtr=/FNGameplayCues/Athena/Sounds/Impacts/Crossbow_Impact_03}}
 {id=WavePlayer_3, properties {soundWaveAssetPtr=/FNGameplayCues/Athena/Sounds/Impacts/Crossbow_Impact_04}}
]
properties {
 attenuationSettings=/Game/Sounds/Attenuation/Fort_Bullet_Impact_Att
 bOverrideConcurrency=true
 maxDistance=5300
 soundClassObject=&Fort_Impacts_PSM
 volumeMultiplier=1.3
 concurrencyOverrides {maxCount=3, resolutionRule=StopOldest}
}
