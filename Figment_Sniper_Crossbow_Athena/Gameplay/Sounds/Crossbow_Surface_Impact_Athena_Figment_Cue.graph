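# Sound cue for the Figment sniper crossbow's surface impact sound.
# A random node picks one of four equally weighted crossbow impact waves and feeds it through a modulator at the root.
# It uses the shared bullet impact attenuation and the impacts sound class, with a boosted cue volume.
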SoundCue /Figment_Sniper_Crossbow_Athena/Gameplay/Sounds/Crossbow_Surface_Impact_Athena_Figment_Cue
root=Modulator_0
nodes [
 {id=Modulator_0, children [Random_0]}
 {id=Random_0, children [WavePlayer_0,WavePlayer_5,WavePlayer_6,WavePlayer_7], properties {weights [1,1,1,1]}}
 {id=WavePlayer_0, properties {soundWaveAssetPtr=&/sfx_crossbow_impact_01_edit}}
 {id=WavePlayer_5, properties {soundWaveAssetPtr=&/sfx_crossbow_impact_02_edit}}
 {id=WavePlayer_6, properties {soundWaveAssetPtr=&/sfx_crossbow_impact_03_edit}}
 {id=WavePlayer_7, properties {soundWaveAssetPtr=&/sfx_crossbow_impact_04_edit}}
]
properties {
 attenuationSettings=/Game/Sounds/Attenuation/Fort_Bullet_Impact_Att
 maxDistance=5300
 soundClassObject=&Fort_Impacts
 volumeMultiplier=6
}
