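# Sound cue for the distant whip-by sound of Figment sniper crossbow projectiles.
# It picks one of four equally weighted whip variations at random, excluded from branch culling, and a modulator pitches it down for variation.
# It plays quietly in the ranged-weapon projectile effects class with distance attenuation overridden off, virtualization disabled, and a shared bullet whiz-by concurrency rule limiting overlap.
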SoundCue /Figment_Sniper_Crossbow_Athena/Gameplay/Sounds/Crossbow_Whip_Far_Figment_Cue
root=Modulator_0
nodes [
 {id=Modulator_0, children [Random_1], properties {pitchMax=0.8, pitchMin=0.7}}
 {
  id=Random_1
  children [WavePlayer_1,WavePlayer_2,WavePlayer_3,WavePlayer_4]
  properties {bShouldExcludeFromBranchCulling=true, weights [1,1,1,1]}
 }
 {id=WavePlayer_1, properties {soundWaveAssetPtr=&/Crossbow_Whip_02_edit}}
 {id=WavePlayer_2, properties {soundWaveAssetPtr=&/Crossbow_Whip_03_edit}}
 {id=WavePlayer_3, properties {soundWaveAssetPtr=&/Crossbow_Whip_04_edit}}
 {id=WavePlayer_4, properties {soundWaveAssetPtr=&/Crossbow_Whip_05_edit}}
]
properties {
 bBypassVolumeScaleForPriority=true
 bOverrideAttenuation=true
 maxDistance=2097152
 priority=0.5
 soundClassObject=&Fort_WeaponRanged_Projectile_Effects
 virtualizationMode=Disabled
 volumeMultiplier=0.2
 attenuationOverrides {bAttenuate=false, stereoSpread=5000}
 concurrencySet [/Game/Sounds/Concurrency_Rules/Athena_BulletWhizBy_Conc]
}
