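# First-person firing sound cue for the Figment sniper crossbow in Athena, using close crossbow shot waves.
# A quality-level switch at the root picks a boosted random of three edited shots, a random of two edited shots, or one original shot shared by the last two levels.
# The boosted branch runs through a mixer whose second input is empty.
# First-person projectile attenuation and gunfire concurrency apply, alongside a ranged-fire sound class and raised overall volume.
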SoundCue /Figment_Sniper_Crossbow_Athena/Gameplay/Sounds/FNS02_Weapon_Crossbow_T02_Athena_1P_Cue
root=QualityLevel_0
nodes [
 {id=Mixer_2, children [Random_11,null], properties {inputVolume [3,0.5]}}
 {id=QualityLevel_0, children [Mixer_2, Random_12, WavePlayer_27, WavePlayer_27]}
 {id=Random_11, children [WavePlayer_0,WavePlayer_1,WavePlayer_2], properties {weights [1,1,1]}}
 {id=Random_12, children [WavePlayer_3,WavePlayer_4], properties {weights [1,1]}}
 {id=WavePlayer_0, properties {soundWaveAssetPtr=&/sfx_crossbow_shot_close_05_edit}}
 {id=WavePlayer_1, properties {soundWaveAssetPtr=&/sfx_crossbow_shot_close_06_edit}}
 {id=WavePlayer_2, properties {soundWaveAssetPtr=&/sfx_crossbow_shot_close_07_edit}}
 {id=WavePlayer_27, properties {soundWaveAssetPtr=/Game/Sounds/Fort_Weapons/Crossbow/sfx_crossbow_shot_close_05}}
 {id=WavePlayer_3, properties {soundWaveAssetPtr=&sfx_crossbow_shot_close_05_edit}}
 {id=WavePlayer_4, properties {soundWaveAssetPtr=&sfx_crossbow_shot_close_06_edit}}
]
properties {
 attenuationSettings=&Fort_Weapon_Projectile_FirstPerson_Athena_Att
 maxDistance=2097152
 soundClassObject=&Fort_WeaponRanged_Fire_1P_PSM
 volumeMultiplier=2
 attenuationOverrides {falloffDistance=5000, attenuationShapeExtents {x=1000}}
 concurrencyOverrides {resolutionRule=PreventNew}
 concurrencySet [&Weapon_Gunfire_FirstPerson_Conc]
}
