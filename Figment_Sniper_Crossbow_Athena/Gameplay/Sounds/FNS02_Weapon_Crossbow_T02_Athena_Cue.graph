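# Sound cue for the firing sound of the Figment sniper crossbow in Athena, layering close and medium-distance shot recordings.
# A mixer blends a close layer with a half-volume medium layer, each picking randomly among evenly weighted waves.
# Each layer has its own projectile attenuation and sound class override: third-person fire for close, distant fire for medium.
# Both random nodes are excluded from branch culling, and the cue uses local and global gunfire concurrency sets.
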SoundCue /Figment_Sniper_Crossbow_Athena/Gameplay/Sounds/FNS02_Weapon_Crossbow_T02_Athena_Cue
root=Mixer_2
nodes [
 {id=Attenuation_5, children [Random_5], properties {attenuationSettings=&Fort_Weapon_Projectile_Close_Athena_Att}}
 {
  id=Attenuation_6
  children [Random_1]
  properties {attenuationSettings=&Fort_Weapon_Projectile_Medium_Mono_Athena_Att}
 }
 {id=Mixer_2, children [SoundClass_5,SoundClass_4], properties {inputVolume [1,0.5]}}
 {
  id=Random_1
  children [WavePlayer_7, WavePlayer_8, WavePlayer_9, WavePlayer_10]
  properties {bShouldExcludeFromBranchCulling=true, weights [1,1,1,1]}
 }
 {id=Random_5, children [WavePlayer_0,WavePlayer_1,WavePlayer_6], properties {bShouldExcludeFromBranchCulling=true, weights [1,1,1]}}
 {
  id=SoundClass_4
  children [Attenuation_6]
  properties {soundClassOverride=&Fort_WeaponRanged_Fire_Distant}
 }
 {
  id=SoundClass_5
  children [Attenuation_5]
  properties {soundClassOverride=&Fort_WeaponRanged_Fire_3P_PSM}
 }
 {id=WavePlayer_0, properties {soundWaveAssetPtr=&/sfx_crossbow_shot_close_05_edit}}
 {id=WavePlayer_1, properties {soundWaveAssetPtr=&/sfx_crossbow_shot_close_06_edit}}
 {id=WavePlayer_10, properties {soundWaveAssetPtr=&/Crossbow_Medium01_edit}}
 {id=WavePlayer_6, properties {soundWaveAssetPtr=&/sfx_crossbow_shot_close_07_edit}}
 {id=WavePlayer_7, properties {soundWaveAssetPtr=&/Crossbow_Medium04_edit}}
 {id=WavePlayer_8, properties {soundWaveAssetPtr=&/Crossbow_Medium03_edit}}
 {id=WavePlayer_9, properties {soundWaveAssetPtr=&/Crossbow_Medium02_edit}}
]
properties {
 maxDistance=10579.2871
 volumeMultiplier=2
 attenuationOverrides {falloffDistance=5000, attenuationShapeExtents {x=1000}}
 concurrencyOverrides {resolutionRule=PreventNew}
 concurrencySet [&Weapon_Gunfire_Local_Conc, &Weapon_Gunfire_Global_Conc]
}
