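# Sound cue for the crossbow equip sound in the Figment sniper crossbow weapon content.
# A fixed short delay feeds a single wave player, which reuses a Fort crossbow reload sample rather than a dedicated equip sound.
# It plays in the weapon foley sound class with reload attenuation, a boosted volume, overridden concurrency and the debug flag enabled.
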SoundCue /Figment_Sniper_Crossbow_Athena/Gameplay/Sounds/Weapon_Crossbow_Equip_Cue
root=Delay_0
nodes [
 {id=Delay_0, children [WavePlayer_27], properties {delayMax=0.15, delayMin=0.15}}
 {id=WavePlayer_27, properties {soundWaveAssetPtr=/Game/Sounds/Fort_Weapons/Crossbow/sfx_crossbow_reload_04}}
]
properties {
 attenuationSettings=&Fort_Weapon_Reload_Att
 bDebug=true
 bOverrideConcurrency=true
 maxDistance=2250
 soundClassObject=&Fort_WeaponFoley
 volumeMultiplier=2
}
